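; Parent spawner actor for itemized multi-chunk prop walls (comments mention an ice wall), built from a data asset plus spawn-time offsets, validity flags and base actors.
; On begin play, authority spawns one chunk per tick via next-tick timers, then links parent chunks.
; Setup sphere-traces to damage overlapping buildings while sparing supporting actors, plays spawn VFX and cue, then attaches props once playlist data is ready.
; It optionally self-destructs on a timer, destroying its props.
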(event EventBeginPlay
  ;[1 1 1] BeginPlay
  (CallFunction|InitializeSpawnedPropDataArray)
  (Variables|PrivateCached|SetCurPropIndex 0)
  (bind _returnvalue (Utilities|Time|SetTimerforNextTickbyEvent (delegate Custom|SpawnNextChunk)))
  (Utilities|Time|ClearandInvalidateTimerbyHandle (Variables|PrivateCached|GetSpawnChunkTimer))
  (Variables|PrivateCached|SetSpawnChunkTimer _returnvalue))

(event Custom|SpawnNextChunk
  ;[1 1 1] BeginPlay
  (Utilities|FlowControl|Sequence
    (:then_0
      (CallFunction|SpawnCurChunk)
      (Math|Integer|IncrementInt (Variables|PrivateCached|GetCurPropIndex)))
    (:then_1
      (if (< (Variables|PrivateCached|GetCurPropIndex) (Utilities|Array|Length (Variables|/ItemizedPropSpawnerParent/Blueprints/DataClass_ItemizedPropSpawnerData.DataClass_ItemizedPropSpawnerData_C|GetNativePropsToSpawn (Variables|Default|GetDataAsset))))
        (bind _returnvalue (Utilities|Time|SetTimerforNextTickbyEvent (delegate Custom|SpawnNextChunk)))
        (Utilities|Time|ClearandInvalidateTimerbyHandle (Variables|PrivateCached|GetSpawnChunkTimer))
        (Variables|PrivateCached|SetSpawnChunkTimer _returnvalue)
        (else
          (CallFunction|AllChunksSpawned))))))

(event Custom|DestroyOverlappingNonSupportingObjects
  ;[1 1 1] Destroy Overlapping Buildings (excluding anything that's supporting the wall)
  (Networking|SwitchHasAuthority
    (:Authority
      (for (_array_element _array_index) (Variables|PublicCached|GetNativeSpawnedPropDataArray)
        (if (.bValid (Utilities|Struct|BreakNativeItemizedPropSpawnerSpawnedPropData _array_element))
          (Utilities|Array|Clear (Variables|Default|GetTempSupportingActors))
          (Utilities|IsValid (.SpawnedActor (Utilities|Struct|BreakNativeItemizedPropSpawnerSpawnedPropData _array_element))
            (:Is%20Valid
              (ItemizedPropSpawner|FindSupportingActors (.SpawnedActor (Utilities|Struct|BreakNativeItemizedPropSpawnerSpawnedPropData _array_element)) (Variables|Default|GetTempSupportingActors) true (Variables|PublicCached|GetSpawnedProps))
              (Utilities|Array|AppendArray (Variables|PrivateCached|GetSupportingActorsForWall) (Variables|Default|GetTempSupportingActors))
              (Utilities|Struct|SetmembersinNativeItemizedPropSpawnerSpawnedPropData (Utilities|Array|Get(aref) (Variables|PublicCached|GetNativeSpawnedPropDataArray) _array_index) (Variables|Default|GetTempSupportingActors))))))
      (Utilities|Array|AppendArray (Variables|PrivateCached|GetActorsToNotDestroy) (Variables|PrivateCached|GetSupportingActorsForWall))
      (for _array_element_2 (Variables|PublicCached|GetNativeSpawnedPropDataArray)
        (bind (_returnvalue_location _returnvalue_rotation) (Math|Transform|BreakTransform (Transformation|GetActorTransform (.SpawnedActor (Utilities|Struct|BreakNativeItemizedPropSpawnerSpawnedPropData _array_element_2)))))
        (bind _outhits (Collision|MultiSphereTraceForObjects (+ _returnvalue_location (* (Math|Vector|GetRightVector _returnvalue_rotation) (Variables|Default|GetDestructionTraceStartOffset))) (+ (+ _returnvalue_location (* (Math|Vector|GetRightVector _returnvalue_rotation) (Variables|Default|GetDestructionTraceStartOffset))) (* (Utilities|Operators|Add (Math|Vector|GetRightVector _returnvalue_rotation) (Math|Vector|MakeVector 0.0 0.0 (Variables|Default|GetDestructionTraceUpAngleOffset))) (ScalableFloat|GetValueatLevel (Variables|/ItemizedPropSpawnerParent/Blueprints/DataClass_ItemizedPropSpawnerData.DataClass_ItemizedPropSpawnerData_C|GetDamageSphereCastDistance (Variables|Default|GetDataAsset))))) (ScalableFloat|GetValueatLevel (Variables|/ItemizedPropSpawnerParent/Blueprints/DataClass_ItemizedPropSpawnerData.DataClass_ItemizedPropSpawnerData_C|GetDamageSphereCastRadius (Variables|Default|GetDataAsset))) (Utilities|Array|MakeArray "ObjectTypeQuery1") false (Variables|PrivateCached|GetActorsToNotDestroy) "ForDuration" true "(R=1.000000,G=0.000000,B=0.000000,A=1.000000)" "(R=0.000000,G=1.000000,B=0.000000,A=1.000000)" 8.0))
        (for _array_element_3 _outhits
          (if (or (.bSuccess (Utilities|Casting|CastToBuildingActor (.HitActor (Collision|BreakHitResult _array_element_3)))) (.bSuccess (Utilities|Casting|CastToSimpleDestructiblePropGroup (.HitActor (Collision|BreakHitResult _array_element_3)))))
            (GameplayEffects|ApplyGameplayEffectToTarget (Ability|GetFortAbilitySystemComponent (Variables|Actor|GetInstigator)) "/ItemizedPropSpawnerParent/Blueprints/GameplayEffects/GE_Damage_ItemizedPropSpawnerParent.GE_Damage_ItemizedPropSpawnerParent_C" (Ability|GetFortAbilitySystemComponent (.HitActor (Collision|BreakHitResult _array_element_3))) 0.0 (Default|CreateDestroyBuildingEffectContext _array_element_3))))))))

(event Custom|SpawnCurChunk
  (bind _curpropindex (Variables|PrivateCached|GetCurPropIndex))
  (bind _nativespawnedpropdataarray (Variables|PublicCached|GetNativeSpawnedPropDataArray))
  (bind _output (Utilities|Array|Get(aref) _nativespawnedpropdataarray _curpropindex))
  (bind (_spawndata _supportingactors _baseactor _spawnedactor _spawnedrelativetransform _zoffset _bvalid) (Utilities|Struct|BreakNativeItemizedPropSpawnerSpawnedPropData _output))
  ;[1 1 1] SpawnCurChunk
  (Networking|SwitchHasAuthority
    (:Authority
      (if _bvalid
        (bind _returnvalue (Game|SpawnActorfromClass (Utilities|Struct|BreakNativeItemizedPropSpawnerPropSpawnData _spawndata) (Math|Transform|ComposeTransforms _spawnedrelativetransform (Transformation|GetActorTransform)) "AlwaysSpawn" "MultiplyWithRoot" 0 0 0 (Variables|Actor|GetInstigator)))
        (Utilities|Struct|SetmembersinNativeItemizedPropSpawnerSpawnedPropData (Utilities|Array|Get(aref) (Variables|PublicCached|GetNativeSpawnedPropDataArray) (Variables|PrivateCached|GetCurPropIndex)) _returnvalue)
        (bind _asprop_spawn_validated (Utilities|Casting|CastToProp_SpawnValidated _returnvalue))
        (Utilities|Array|Add (Variables|PublicCached|GetSpawnedProps) _asprop_spawn_validated)))))

(event Custom|AllChunksSpawned
  ;[1 1 1] AllChunksSpawned
  (Utilities|FlowControl|Sequence
    (:then_0
      (CallFunction|AssignParentChunks))
    (:then_1
      (CallFunction|FinishSetup))))

(event Custom|AssignParentChunks
  (bind _spawnedprops (Variables|PublicCached|GetSpawnedProps))
  (bind _output (Utilities|Array|Get(acopy) _spawnedprops 1))
  (bind _output_2 (Utilities|Array|Get(acopy) _spawnedprops 2))
  (bind _output_3 (Utilities|Array|Get(acopy) _spawnedprops 3))
  ;[1 1 1] AssignParentChunks
  (Networking|SwitchHasAuthority
    (:Authority
      (Utilities|FlowControl|Sequence
        (:then_0
          (Utilities|IsValid _output
            (:Is%20Valid
              (Variables|/Script/ItemizedPropSpawnerParentRuntime.ItemizedPropSpawnerWallChunk|SetParentChunk _output (Utilities|Array|Get(acopy) _spawnedprops)))))
        (:then_1
          (Utilities|IsValid _output_2
            (:Is%20Valid
              (Variables|/Script/ItemizedPropSpawnerParentRuntime.ItemizedPropSpawnerWallChunk|SetParentChunk _output_2 _output))))
        (:then_2
          (Utilities|IsValid _output_2
            (:Is%20Valid
              (Variables|/Script/ItemizedPropSpawnerParentRuntime.ItemizedPropSpawnerWallChunk|SetParentChunk _output_2 _output_3))))
        (:then_3
          (Utilities|IsValid _output_3
            (:Is%20Valid
              (Variables|/Script/ItemizedPropSpawnerParentRuntime.ItemizedPropSpawnerWallChunk|SetParentChunk _output_3 (Utilities|Array|Get(acopy) _spawnedprops 4)))))))))

(event Custom|FinishSetup
  ;[1 1 1] FinishSetup
  (Utilities|FlowControl|Sequence
    (:then_0
      (Utilities|Array|Clear (Variables|PrivateCached|GetSupportingActorsForWall))
      (Utilities|Array|Clear (Variables|PrivateCached|GetActorsToNotDestroy))
      (for _array_element (Variables|PublicCached|GetSpawnedProps)
        (Utilities|IsValid _array_element
          (:Is%20Valid
            (Utilities|Array|Add (Variables|PrivateCached|GetActorsToNotDestroy) _array_element)))))
    (:then_1
      (CallFunction|DestroyOverlappingNonSupportingObjects)
      ;;[1 0.3445 0.308] VFX: Activate Ice wall
      (Niagara|SpawnSystematLocation (Variables|Default|GetSpawnVFX) (Transformation|GetActorLocation) (Transformation|GetActorRotation)))
    (:then_2
      (GameplayCue|ExecuteGameplayCue self (Variables|/ItemizedPropSpawnerParent/Blueprints/DataClass_ItemizedPropSpawnerData.DataClass_ItemizedPropSpawnerData_C|GetGC_Spawn (Variables|Default|GetDataAsset)) (Class|AthenaFunctionLibrary|MakeSimpleFortEffectContext self self)))
    (:then_3
      (Networking|SwitchHasAuthority
        (:Authority
          (Class|FortAsyncActionCurrentPlaylistReady|PlaylistDataReadyAsync
            (:OnReady
              (for _array_element_2 (Variables|PublicCached|GetNativeSpawnedPropDataArray)
                (if (CallFunction|SpawnedPropDataIsValid _array_element_2)
                  (ItemizedPropSpawner|AttachActors (.SpawnedActor (Utilities|Struct|BreakNativeItemizedPropSpawnerSpawnedPropData _array_element_2)) (Variables|PublicCached|GetSpawnedProps) (.SupportingActors (Utilities|Struct|BreakNativeItemizedPropSpawnerSpawnedPropData _array_element_2)) (.BaseActor (Utilities|Struct|BreakNativeItemizedPropSpawnerSpawnedPropData _array_element_2))))))))))
    (:then_4
      (if (Class|AthenaFunctionLibrary|RowtoBool (Variables|/ItemizedPropSpawnerParent/Blueprints/DataClass_ItemizedPropSpawnerData.DataClass_ItemizedPropSpawnerData_C|GetDestroyOverTime_Enabled (Variables|Default|GetDataAsset)))
        (Utilities|Time|SetTimerbyEvent (delegate Custom|DestroySelf) (ScalableFloat|GetValueatLevel (Variables|/ItemizedPropSpawnerParent/Blueprints/DataClass_ItemizedPropSpawnerData.DataClass_ItemizedPropSpawnerData_C|GetDestroyOverTime_Time (Variables|Default|GetDataAsset))))))))

(event Custom|DestroySelf
  ;[1 1 1] FinishSetup
  (Actor|DestroyActor self))

(event EventDestroyed
  ;[1 1 1] FinishSetup
  (for _array_element (Variables|PublicCached|GetSpawnedProps)
    (Utilities|IsValid _array_element
      (:Is%20Valid
        (Actor|DestroyActor _array_element)))))

(fn CheckForBuilds
  (params (Array :type Array%20of%20Actor%20Object%20References))
  (for _array_element Array
    (Utilities|Casting|CastToBuildingWall _array_element)
    (LocalVariables|SetHasBuildingActor_ true))
  (return (LocalVariables|GetHasBuildingActor_)))

(fn CreateDestroyBuildingEffectContext
  (params (InHitResult :type Hit%20Result%20Structure))
  (bind _output_get (LocalVariables|SetEffectContextHandle (Class|AthenaFunctionLibrary|MakeSimpleFortEffectContext (Variables|Actor|GetInstigator) self)))
  (Ability|EffectContext|AddHitResult _output_get InHitResult true)
  (return (LocalVariables|GetEffectContextHandle)))

(fn InitializeSpawnedPropDataArray
  (for (_array_element _array_index) (Variables|/ItemizedPropSpawnerParent/Blueprints/DataClass_ItemizedPropSpawnerData.DataClass_ItemizedPropSpawnerData_C|GetNativePropsToSpawn (Variables|Default|GetDataAsset))
    (bind (_location _rotation _scale) (Math|Transform|BreakTransform (.RelativeTransform (Utilities|Struct|BreakNativeItemizedPropSpawnerPropSpawnData _array_element))))
    (bind (_location_x _location_y _location_z) (Math|Vector|BreakVector _location))
    (Utilities|Array|Add (Variables|PublicCached|GetNativeSpawnedPropDataArray) (Utilities|Struct|MakeNativeItemizedPropSpawnerSpawnedPropData :SpawnData _array_element :BaseActor (Utilities|Array|Get(acopy) (Variables|PassInOnSpawn|GetActorsBeingUsedAsBase) _array_index) :SpawnedRelativeTransform (Math|Transform|MakeTransform (Math|Vector|MakeVector _location_x _location_y (+ (+ (* (Utilities|Array|Get(acopy) (Variables|PassinonSpawn|GetOffsets) _array_index) -1.0) (ScalableFloat|GetValueatLevel (Variables|/ItemizedPropSpawnerParent/Blueprints/DataClass_ItemizedPropSpawnerData.DataClass_ItemizedPropSpawnerData_C|GetGlobal%20Z%20Offset (Variables|Default|GetDataAsset)))) _location_z)) _rotation _scale) :ZOffset (+ (* (Utilities|Array|Get(acopy) (Variables|PassinonSpawn|GetOffsets) _array_index) -1.0) (ScalableFloat|GetValueatLevel (Variables|/ItemizedPropSpawnerParent/Blueprints/DataClass_ItemizedPropSpawnerData.DataClass_ItemizedPropSpawnerData_C|GetGlobal%20Z%20Offset (Variables|Default|GetDataAsset)))) :bValid (Utilities|Array|Get(acopy) (Variables|PassinonSpawn|GetValidity) _array_index)))))

(fn SpawnedPropDataIsValid
  (params (Spawned%20Prop%20Data :type Native%20Itemized%20Prop%20Spawner%20Spawned%20Prop%20Data%20Structure))
  (bind (_spawndata _supportingactors _baseactor _spawnedactor _spawnedrelativetransform _zoffset _bvalid) (Utilities|Struct|BreakNativeItemizedPropSpawnerSpawnedPropData Spawned%20Prop%20Data))
  (return (and (Utilities|IsValid _spawnedactor :!class /Script/Engine.KismetSystemLibrary) _bvalid)))

(fn UpdateLocalTransforms
  (for (_ _array_index) (Variables|PassinonSpawn|GetOffsets)
    (if (Utilities|Array|Get(acopy) (Variables|PassinonSpawn|GetValidity) _array_index)
      (bind (_output_location _output_rotation _output_scale) (Math|Transform|BreakTransform (Utilities|Array|Get(acopy) (Variables|/ItemizedPropSpawnerParent/Blueprints/DataClass_ItemizedPropSpawnerData.DataClass_ItemizedPropSpawnerData_C|GetPropRelativeTransforms (Variables|Default|GetDataAsset)) _array_index)))
      (bind (_output_location_x _output_location_y _output_location_z) (Math|Vector|BreakVector _output_location))
      (Utilities|Array|Add (Variables|ToBeDeleted|GetSpawnedPropLocalTransforms) (Math|Transform|MakeTransform (Math|Vector|MakeVector _output_location_x _output_location_y (Utilities|Operators|Add _output_location_z (* (Utilities|Array|Get(acopy) (Variables|PassinonSpawn|GetOffsets) _array_index) -1.0) -1e+02)) _output_rotation _output_scale))
      (else
        (Utilities|Array|Add (Variables|ToBeDeleted|GetSpawnedPropLocalTransforms) (Math|Transform|MakeTransform "0, 0, 0"))))))

(fn ConstructionScript)
