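; Parent weapon Blueprint for items that spawn a multi-prop wall at a targeted spot and launch nearby pawns and vehicles.
; While equipped, a next-tick timer loop runs native targeting on server and owning client, updating fire blocking and optional preview meshes.
; Authority-side firing spawns and knocks back after short delays, or on landing if falling.
; Data-asset tuning feeds native targeting; gameplay effects apply cooldown, and a controller component caps live walls.
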(event FortWeapon|EventOnUnEquippedAndDetached
  (Variables|Default|SetCachedTargetedLocation "0, 0, 0")
  (if (or (Utilities|Enum|Equal(Enum) (Networking|GetLocalRole (Variables|Actor|GetInstigator)) "ROLE_Authority") (Utilities|Enum|Equal(Enum) (Networking|GetLocalRole (Variables|Actor|GetInstigator)) "ROLE_AutonomousProxy"))
    (Utilities|Time|ClearandInvalidateTimerbyHandle (Variables|Default|GetUpdateTargetTimer))))

(event FortWeapon|EventOnEquippedAndAttached
  (if (or (Utilities|Enum|Equal(Enum) (Networking|GetLocalRole (Variables|Actor|GetInstigator)) "ROLE_Authority") (Utilities|Enum|Equal(Enum) (Networking|GetLocalRole (Variables|Actor|GetInstigator)) "ROLE_AutonomousProxy"))
    (bind _returnvalue (Utilities|Time|SetTimerforNextTickbyEvent (delegate Custom|UpdateTarget)))
    (Utilities|Time|ClearandInvalidateTimerbyHandle (Variables|Default|GetUpdateTargetTimer))
    (Variables|Default|SetUpdateTargetTimer _returnvalue)))

(event Custom|UpdateTarget
  (bind _result (CallFunction|CanbeTriggered))
  (Utilities|FlowControl|Sequence
    (:then_0
      (if (CallFunction|CanUpdateTarget)
        (bind (_outtargetlocation _outfailurereason _outvalidity _returnvalue) (ItemizedPropSpawner|NativeUpdateTarget (Variables|Default|GetNativeData)))
        (CallFunction|OnPreviewingChanged (Variables|Default|GetPreviewing))
        (Variables|Default|SetChunkValidity _outvalidity)
        (Utilities|FlowControl|Sequence
          (:then_0
            (if (!= (Variables|Default|GetWasFiringEnabled) (CallFunction|CanbeTriggered))
              (Variables|Default|SetWasFiringEnabled _result)
              (Default|CallWeaponBlockChanged (not _result))))
          (:then_1
            (bind _output_get (Variables|Default|SetbHasValidTarget _returnvalue))
            (if _output_get
              (Variables|Default|SetTargetedLocation _outtargetlocation))))))
    (:then_1
      (bind _returnvalue_2 (Utilities|Time|SetTimerforNextTickbyEvent (delegate Custom|UpdateTarget)))
      (Utilities|Time|ClearandInvalidateTimerbyHandle (Variables|Default|GetUpdateTargetTimer))
      (Variables|Default|SetUpdateTargetTimer _returnvalue_2))))

(event EventBeginPlay
  (Utilities|FlowControl|Sequence
    (:then_0
      (CallFunction|SetUpPreviewMeshes)
      (CallFunction|SetUpWeaponData))
    (:then_1
      ;[1 1 1] Manage Preview State (disabled by default, enabled in children)
      (if (Variables|ChildOptions|GetShouldShowPreview)
        (Ability|Tasks|WaitGameplayTagAddToActor (Variables|Default|GetPlayer) (Variables|Default|GetTag_Previewing)
          (:then
            (Ability|Tasks|WaitGameplayTagRemoveFromActor (Variables|Default|GetPlayer) (Variables|Default|GetTag_Previewing)
              (:Removed
                (CallFunction|OnPreviewingChanged false))))
          (:Added
            (CallFunction|OnPreviewingChanged true)))))))

(event Custom|On%20Previewing%20Changed
  (params (IsPreviewing :type Boolean))
  ;[1 1 1] Manage Preview State (disabled by default, enabled in children)
  (if (Variables|ChildOptions|GetShouldShowPreview)
    (bind _output_get (Variables|Default|SetPreviewing IsPreviewing))
    (Utilities|Struct|SetmembersinItemizedPropSpawnerWeaponData (Variables|Default|GetNativeData) _output_get)
    (for (_array_element _array_index) (Variables|Default|GetPreviewMeshes)
      (Rendering|SetVisibility _array_element (Math|Boolean|ANDBoolean (CallFunction|SafeGet _array_index (Variables|Default|GetChunkValidity)) (Variables|Default|GetPreviewing) (Pawn|IsLocallyControlled (Variables|Default|GetPlayer)) (Variables|Default|GetbHasValidTarget)) false))))

(event Custom|CallWeaponFireEvent_Multicast
  ;[1 1 1] For animation
  (Default|CallWeaponFire))

(event Custom|OnWeaponFired
  (bind _player (Variables|Default|GetPlayer))
  (bind _returnvalue (Ability|GetFortAbilitySystemComponent _player))
  ;[1 1 1] On Fired (spawn wall)
  (Variables|Default|SetCachedTargetedLocation (Variables|Default|GetTargetedLocation))
  (Networking|SwitchHasAuthority
    (:Authority
      (if (CallFunction|IsNotCurrentlySpawningWall)
        (CallFunction|CallWeaponFireEventMulticast)
        (if (GameplayTags|HasMatchingGameplayTag (Variables|Default|GetPlayer) "(TagName=\"MovementMode.Falling\")")
          ;;[1 1 1] Spawn wall on landing
          (Variables|Default|SetIsAttackFromAirPending true)
          (bind _returnvalue_2 (GameplayEffects|ApplyGameplayEffectToSelf :self _returnvalue :GameplayEffectClass (Variables|Default|GetGE_FallingIntoSpawn) :Level 0.0))
          (Variables|Default|SetGE_FallingWithIce _returnvalue_2)
          (GameplayEffects|ApplyGameplayEffectToSelf :self _returnvalue :GameplayEffectClass (Variables|Default|GetGE_BlockWeaponSwitchingWhileFallingIntoSpawn) :Level 0.0)
          (Ability|Tasks|WaitGameplayTagRemoveFromActor (Variables|Default|GetPlayer) "(TagName=\"MovementMode.Falling\")" true
            (:Removed
              (GameplayEffects|RemoveActiveEffectswithGrantedTags (Ability|GetFortAbilitySystemComponent (Variables|Default|GetPlayer)) "(GameplayTags=((TagName=\"Granted.Athena.SereneYard.BlockChangingEquipmentActive\")))")
              (if (CallFunction|ShouldSpawnWallAfterFallingActivation)
                (CallFunction|DoLaunch)
                (CallFunction|SpawnWall)
                (else
                  (FailedAction|UnabletoPerformAction (Variables|Default|GetPlayer) false (GameplayTags|MakeGameplayTagContainerfromTag "(TagName=\"Athena.Ability.Failed.SereneYard.SpawnBlocked\")")) ; show reason for failing on screen if one is given
                ))
              (GameplayEffects|RemoveActiveGameplayEffect (Ability|GetFortAbilitySystemComponent (Variables|Default|GetPlayer)) (Variables|Default|GetGE_FallingWithIce) -1)
              (Variables|Default|SetIsAttackFromAirPending false)))
          (else
            (GameplayEffects|ApplyGameplayEffectToSelf :self (Ability|GetFortAbilitySystemComponent (Variables|Default|GetPlayer)) :GameplayEffectClass (Variables|Default|GetGE_BlockSwitchingWeaponsWhileDoingSpawnAnim) :Level 0.0)
            (bind _returnvalue_3 (Utilities|Time|SetTimerbyEvent (delegate Custom|SpawnWall) 0.4))
            (Variables|Default|SetSpawnWallTimer _returnvalue_3)
            ;[1 1 1] Launch Enemies
            (Utilities|Time|SetTimerbyEvent (delegate Custom|DoLaunch) (ScalableFloat|GetValueatLevel (Variables|/ItemizedPropSpawnerParent/Blueprints/DataClass_ItemizedPropSpawnerData.DataClass_ItemizedPropSpawnerData_C|GetLaunchDelay (Variables|Default|GetDataAsset))))))))))

(event Custom|DoLaunch
  (bind _knockback_traceoffsetcenter (Variables|Default|GetKnockback_TraceOffsetCenter))
  (bind _returnvalue (Utilities|GetIceLocationfromOffset _knockback_traceoffsetcenter))
  (bind _instigator (Variables|Actor|GetInstigator))
  (bind _array (Utilities|Array|MakeArray _instigator))
  ;[1 1 1] Launch Enemies
  (bind _outhits (Collision|MultiSphereTraceByChannel _returnvalue (Utilities|GetIceLocationfromOffset (Variables|Default|GetKnockback_TraceOffsetLeft)) (Variables|Default|GetKnockback_TraceRadius) "TraceTypeQuery1" false _array "ForDuration"))
  (CallFunction|LaunchActors _outhits)
  (bind _outhits_2 (Collision|MultiSphereTraceByChannel _returnvalue (Utilities|GetIceLocationfromOffset (Variables|Default|GetKnockback_TraceOffsetRight)) (Variables|Default|GetKnockback_TraceRadius) "TraceTypeQuery1" false _array "ForDuration"))
  (CallFunction|LaunchActors _outhits_2))

(event Custom|SpawnWall
  (bind (_returnvalue_roll _returnvalue_pitch _returnvalue_yaw) (Math|Rotator|BreakRotator (Utilities|GetRotationOfSpawn)))
  ;[1 1 1] On Fired (spawn wall)
  (GameplayEffects|RemoveActiveEffectswithGrantedTags (Ability|GetFortAbilitySystemComponent (Variables|Default|GetPlayer)) "(GameplayTags=((TagName=\"Granted.Athena.SereneYard.BlockChangingEquipmentActive\")))")
  (bind (_outvalidity _outoffsets _outactorsbeingusedasbase) (ItemizedPropSpawner|NativeCheckSpawnValidity (Variables|Default|GetNativeData) (CallFunction|GetTargetedLocation) _returnvalue_yaw true))
  (if (> (Fortnite|GetNumericCVar "SereneYardFunctionLibrary.ShowDebugSpawningView") 0)
    (Rendering|Debug|DrawDebugSphere (CallFunction|GetTargetedLocation) 3e+01 10 "(R=0.000000,G=1.000000,B=0.000000,A=1.000000)" 1e+01 2.0))
  (bind _returnvalue (Game|SpawnActorfromClass (Variables|Default|GetBGA_PropSpawner) (Math|Transform|MakeTransform (CallFunction|GetTargetedLocation) (Utilities|GetRotationOfSpawn)) "Undefined" "MultiplyWithRoot" 0 (select (Class|AthenaFunctionLibrary|RowtoBool (Variables|ScalableFloats|GetSF_UseSupportingActorsArrayFromSpawnValidityCheck)) _outactorsbeingusedasbase (Utilities|Array|MakeArray (Variables|Actor|GetInstigator))) _outoffsets _outvalidity 0 (Variables|Actor|GetInstigator)))
  (CallFunction|EnforceSpawnLimit _returnvalue)
  (Variables|Default|SetLastSpawnedBGA _returnvalue)
  (bind _output_get (Variables|Default|SetStopped false))
  (Utilities|Struct|SetmembersinItemizedPropSpawnerWeaponData (Variables|Default|GetNativeData) _output_get)
  (GameplayEffects|ApplyGameplayEffectToSelf :self (Ability|GetFortAbilitySystemComponent (Variables|Default|GetPlayer)) :GameplayEffectClass (Variables|Default|GetGE_Cooldown) :Level 0.0)
  (GameplayEffects|ApplyGameplayEffectToSelf :self (Ability|GetFortAbilitySystemComponent (Variables|Default|GetPlayer)) :GameplayEffectClass (Variables|Default|GetGE_OnSpawnedGE) :Level 0.0))

(event Custom|LaunchActors
  (params (Array :type Array%20of%20Hit%20Result%20Structures))
  (bind _player (Variables|Default|GetPlayer))
  (bind _returnvalue (Vehicle|GetVehicle _player))
  (bind _player_2 (Variables|Default|GetPlayer))
  (bind _dataasset (Variables|Default|GetDataAsset))
  (bind _returnvalue_2 (Class|DataClassItemizedPropSpawnerData|GetLaunchVelocity _dataasset _player_2 true))
  (bind _player_3 (Variables|Default|GetPlayer))
  ;[1 1 1] Launch Enemies
  (for _array_element Array
    (bind _asfort_pawn (Utilities|Casting|CastToFortPawn (.HitActor (Collision|BreakHitResult _array_element)))
      (:then
        (if (Math|Boolean|ORBoolean (Class|AthenaFunctionLibrary|RowtoBool (Variables|/ItemizedPropSpawnerParent/Blueprints/DataClass_ItemizedPropSpawnerData.DataClass_ItemizedPropSpawnerData_C|GetPerformEarlyLargeAOELaunchOnTeammates (Variables|Default|GetDataAsset))) (not (Team|OnSameTeam _player_3 _asfort_pawn)) (not (.bSuccess (Utilities|Casting|CastToPlayerPawn_Athena _asfort_pawn))))
          (Pawn|Character|LaunchCharacterJump _asfort_pawn (Class|DataClassItemizedPropSpawnerData|GetLaunchVelocity (Variables|Default|GetDataAsset) (Variables|Default|GetPlayer) false) true true (or (Team|OnSameTeam _player_3 _asfort_pawn) (Class|AthenaFunctionLibrary|RowtoBool (Variables|ScalableFloats|GetSF_ShouldApplyFallImmunityToKnockbackVictims))) true)
          (CallFunction|ExecuteImpulseGC _asfort_pawn)
          (Ability|SendGameplayEventtoActor (Variables|Default|GetPlayer) "(TagName=\"Event.Athena.ItemizedPropSpawner.LaunchedPlayer\")")))
      (:CastFailed
        (bind _asfort_athena_vehicle (Utilities|Casting|CastToFortAthenaVehicle (.HitActor (Collision|BreakHitResult _array_element))))
        (Audio|BroadcastAbilityImpactatLocation 0 _asfort_athena_vehicle _returnvalue_2 777.0) ; Launch Vehicle\n(Node is disguised for anti-cheat reasons)
        (FortPhysicsSimulationLibrary|SetImpactInstigatorforPhysicsObject _asfort_athena_vehicle (Variables|Default|GetPlayer))
        (CallFunction|ExecuteImpulseGC _asfort_athena_vehicle))))
  (if (and (Vehicle|IsinVehicle (Variables|Default|GetPlayer)) (not (Variables|Default|GetPreviewing)))
    (Audio|BroadcastAbilityImpactatLocation 0 _returnvalue _returnvalue_2 777.0) ; Launch Vehicle\n(Node is disguised for anti-cheat reasons)
    (FortPhysicsSimulationLibrary|SetImpactInstigatorforPhysicsObject _returnvalue (Variables|Default|GetPlayer))
    (CallFunction|ExecuteImpulseGC _returnvalue)))

(event Custom|ExecuteImpulseGC
  (params (TargetActor :type Actor%20Object%20Reference))
  ;[1 1 1] Launch Enemies
  (Utilities|DoOncePerFrame
    (:%20%20
      (GameplayCue|ExecuteGameplayCue TargetActor (Variables|Default|GetGC_ImpulsedSomeone)))))

(event Custom|EnforceSpawnLimit
  (params (NewWall :type BGA%20Multi%20Prop%20Spawner%20Parent%20Object%20Reference))
  (bind _player (Variables|Default|GetPlayer))
  (bind _returnvalue (Pawn|GetController _player))
  (bind _returnvalue_2 (Actor|GetComponentbyClass _returnvalue "/Script/ItemizedPropSpawnerParentRuntime.FortControllerComponent_ItemizedPropSpawner"))
  (ItemizedPropSpawner|AddTrackedInstance _returnvalue_2 NewWall)
  (if (> (ItemizedPropSpawner|GetTrackedInstanceCount _returnvalue_2) (ScalableFloat|GetValueasInteger (Variables|ScalableFloats|GetSF_MaxSpawnedWalls)))
    (Actor|DestroyActor (ItemizedPropSpawner|GetOldestTrackedInstance _returnvalue_2))
    (ItemizedPropSpawner|RemoveOldestTrackedInstance _returnvalue_2)))

(fn CanBeTriggered
  (return (Math|Boolean|ANDBoolean (or (and (Variables|Default|GetbHasValidTarget) (not (AI|Components|NavMovement|IsFalling (Variables|/Script/Engine.Character|GetCharacterMovement (Variables|Default|GetPlayer))))) (and (AI|Components|NavMovement|IsFalling (Variables|/Script/Engine.Character|GetCharacterMovement (Variables|Default|GetPlayer))) (not (Variables|Default|GetIsAttackFromAirPending)))) (not (Math|Boolean|ORBoolean (Utilities|CheckifDisabledbyTag (Variables|ScalableFloats|GetSF_DisabledInVehicle) (GameplayTags|MakeLiteralGameplayTagContainer "(GameplayTags=((TagName=\"Gameplay.Action.Player.AthenaInVehicle\")))")) (Utilities|CheckifDisabledbyTag (Variables|ScalableFloats|GetSF_DisabledWhileRiding) (GameplayTags|MakeLiteralGameplayTagContainer "(GameplayTags=((TagName=\"Granted.Riding.Player.IsRiding\")))")) (Utilities|CheckifDisabledbyTag (Variables|ScalableFloats|GetSF_DisabledWhileZiplining) (GameplayTags|MakeLiteralGameplayTagContainer "(GameplayTags=((TagName=\"MovementMode.ZipliningOnSpline\"),(TagName=\"MovementMode.Ziplining\")))")) (Utilities|CheckifDisabledbyTag (Variables|ScalableFloats|GetSF_DisabledWhileGrinding) (GameplayTags|MakeLiteralGameplayTagContainer "(GameplayTags=((TagName=\"MovementMode.Grinding\")))")) (Utilities|CheckifDisabledbyTag (Variables|ScalableFloats|GetSF_DisabledWhileSliding) (GameplayTags|MakeLiteralGameplayTagContainer "(GameplayTags=((TagName=\"MovementMode.Sliding\")))")) (Utilities|CheckifDisabledbyTag (Variables|ScalableFloats|GetSF_DisabledWhileFalling) (GameplayTags|MakeLiteralGameplayTagContainer "(GameplayTags=((TagName=\"MovementMode.Falling\")))")) (Utilities|CheckifDisabledbyTag (Variables|ScalableFloats|GetSF_DisabledWhileSwimming) (GameplayTags|MakeLiteralGameplayTagContainer "(GameplayTags=((TagName=\"MovementMode.SurfaceSwimming\")))")))) (not (GameplayTags|HasMatchingGameplayTag (Variables|Default|GetPlayer) (Variables|Default|GetTag_Cooldown))) (not (Utilities|Time|IsTimerActivebyHandle (Variables|Default|GetSpawnWallTimer)))) (Variables|Default|GetbHasValidTarget)))

(fn CanUpdateTarget
  (return (and (Utilities|IsValid (Utilities|Struct|BreakItemizedPropSpawnerWeaponData (Variables|Default|GetNativeData)) :!class /Script/Engine.KismetSystemLibrary) (not (GameplayTags|HasMatchingGameplayTag (Variables|Default|GetPlayer) (Variables|Default|GetTag_Cooldown))))))

(fn CheckForBuilds
  (params (Array :type Array%20of%20Actor%20Object%20References))
  (for _array_element Array
    (Utilities|Casting|CastToBuildingWall _array_element)
    (LocalVariables|SetHasBuildingActor_ true))
  (return (LocalVariables|GetHasBuildingActor_)))

(fn CheckSpawnValidity
  (params (DebugDrawTime :type Float%20%28double-precision%29))
  (bind _currentindex (LocalVariables|GetCurrentIndex))
  (bind _previewmeshes (Variables|Default|GetPreviewMeshes))
  (bind _output (Utilities|Array|Get(acopy) _previewmeshes _currentindex))
  (bind _icerelativetransforms (Variables|Default|GetIceRelativeTransforms))
  ;[1 0 0] This function has been nativized as NativeCheckSpawnValidity, please use that instead. If you need to make changes contact Brian Zaugg.
  (bind _keys (Utilities|Map|Keys (Variables|Default|GetIceRelativeTransforms)))
  (for (_array_element _array_index) _keys
    (LocalVariables|SetCurrentIndex _array_index)
    (bind _hit (CallFunction|TraceforSpawning (+ (Utilities|GetIceLocationfromOffset (+ (.location (Utilities|Map|Find _icerelativetransforms _array_element)) (* (Math|Vector|GetRightVector (.rotation (Utilities|Map|Find _icerelativetransforms _array_element))) 1.3e+02))) "0.000000,0.000000,200.000000") (+ (Utilities|GetIceLocationfromOffset (+ (.location (Utilities|Map|Find _icerelativetransforms _array_element)) (* (Math|Vector|GetRightVector (.rotation (Utilities|Map|Find _icerelativetransforms _array_element))) 1.3e+02))) "0.000000,0.000000,-1200.000000")))
    (Utilities|Array|Add (LocalVariables|GetActorsBeingUsedAsBase) (.HitActor (Collision|BreakHitResult _hit)))
    (Transformation|SetWorldLocationAndRotation _output (.ImpactPoint (Collision|BreakHitResult _hit)) (Math|Rotator|CombineRotators (Math|Rotator|CombineRotators (Math|Rotator|MakeRotator (.yaw (.rotation (Utilities|Map|Find _icerelativetransforms _array_element)))) "-90.000000,0.000000,0.000000") (Utilities|GetRotationOfSpawn)) false true)
    (Rendering|Material|SetMaterial _output 0 "/Game/Characters/Player/Male/Medium/Bodies/M_MED_Bounds/Meshes/Bounds_Transparent_Red_Mat.Bounds_Transparent_Red_Mat")
    (Utilities|Map|Add (LocalVariables|GetValidityMap) _array_element 0)
    (Utilities|Map|Add (LocalVariables|GetOffsetsMap) _array_element (- (.Distance (Collision|BreakHitResult _hit)) 2e+02)))
  (return (LocalVariables|GetValidityMap) (LocalVariables|GetOffsetsMap) (LocalVariables|GetActorsBeingUsedAsBase)))

(fn FindFirstValidHit
  (params (InputPin :type Array%20of%20Hit%20Result%20Structures))
  ;[1 0 0] This function has been nativized as NativeFindFirstValidHit, please use that instead. If you need to make changes contact Brian Zaugg.
  (for _array_element InputPin
    (LocalVariables|SetThisOnesInvalid false)
    (bind _array_element_2 (Utilities|Array|ForEachLoopwithBreak (Variables|Default|GetClassesToIgnoreWhenTracing))
      (:LoopBody
        (if (Utilities|IsA (.HitActor (Collision|BreakHitResult _array_element)) _array_element_2)
          (LocalVariables|SetThisOnesInvalid true)))
      (:Completed
        _))
    (if (not (LocalVariables|GetThisOnesInvalid))
      (return _array_element))))

(fn GetTargetedLocation
  (if (or (Variables|Default|GetIsAttackFromAirPending) (Variables|/Script/FortniteGame.FortPawn|GetbIsSliding (Variables|Default|GetPlayer)))
    (return (Variables|Default|GetTargetedLocation))
    (else
      (return (Variables|Default|GetCachedTargetedLocation)))))

(fn HideReticle
  (bind _asfort_player_controller_zone (Utilities|Casting|CastToFortPlayerControllerZone (Game|GetInstigatorController)))
  (HUD|SetHUDElementVisibilitywithContext _asfort_player_controller_zone (Variables|Scope|GetReticleHUDElementTags) true "None" ""))

(fn IsNotCurrentlySpawningWall
  (return (and (not (Utilities|Time|IsTimerActivebyHandle (Variables|Default|GetSpawnWallTimer))) (not (Variables|Default|GetIsAttackFromAirPending)))))

(fn SafeGet
  (params (Value :type Integer) (Array :type Array%20of%20Booleans))
  (if (Utilities|Array|IsEmpty Array)
    (return false)
    (elif (Math|Integer|InRange(Integer) Value 0 (Utilities|Array|LastIndex Array))
      (return (Utilities|Array|Get(acopy) Array Value))
      (else
        (return false)))))

(fn SetUpPreviewMeshes
  ;[1 1 1] TODO code currently crashes if this array is empty
  (Variables|Default|SetPreviewMeshes (Utilities|Array|MakeArray 0 0 0 0 0)))

(fn SetUpWeaponData
  (bind (_proprelativetransforms _chunklocationtopropindexmap) (ItemizedPropSpawner|BuildPropDataforWeaponData (Variables|/ItemizedPropSpawnerParent/Blueprints/DataClass_ItemizedPropSpawnerData.DataClass_ItemizedPropSpawnerData_C|GetNativePropsToSpawn (Variables|Default|GetDataAsset))))
  (bind _asplayer_pawn_athena (Utilities|Casting|CastToPlayerPawn_Athena (Variables|Actor|GetInstigator)))
  (bind _output_get (Variables|Default|SetPlayer _asplayer_pawn_athena))
  (Utilities|Struct|SetmembersinItemizedPropSpawnerWeaponData (Variables|Default|GetNativeData) _output_get self (ScalableFloat|GetValueatLevel (Variables|/ItemizedPropSpawnerParent/Blueprints/DataClass_ItemizedPropSpawnerData.DataClass_ItemizedPropSpawnerData_C|GetMaxRange_ADSing (Variables|Default|GetDataAsset))) (ScalableFloat|GetValueatLevel (Variables|/ItemizedPropSpawnerParent/Blueprints/DataClass_ItemizedPropSpawnerData.DataClass_ItemizedPropSpawnerData_C|GetMaxRange_Base (Variables|Default|GetDataAsset))) (ScalableFloat|GetValueatLevel (Variables|/ItemizedPropSpawnerParent/Blueprints/DataClass_ItemizedPropSpawnerData.DataClass_ItemizedPropSpawnerData_C|GetMinDistance (Variables|Default|GetDataAsset))) (ScalableFloat|GetValueatLevel (Variables|/ItemizedPropSpawnerParent/Blueprints/DataClass_ItemizedPropSpawnerData.DataClass_ItemizedPropSpawnerData_C|GetMinDistFromIndestructible (Variables|Default|GetDataAsset))) (ScalableFloat|GetValueasBool (Variables|/ItemizedPropSpawnerParent/Blueprints/DataClass_ItemizedPropSpawnerData.DataClass_ItemizedPropSpawnerData_C|GetEnableBlockedByIndestructible (Variables|Default|GetDataAsset))) (ScalableFloat|GetValueatLevel (Variables|/ItemizedPropSpawnerParent/Blueprints/DataClass_ItemizedPropSpawnerData.DataClass_ItemizedPropSpawnerData_C|GetPlayerForwardVectorScale (Variables|Default|GetDataAsset))) (ScalableFloat|GetValueatLevel (Variables|/ItemizedPropSpawnerParent/Blueprints/DataClass_ItemizedPropSpawnerData.DataClass_ItemizedPropSpawnerData_C|GetInitialTracePullbackDist (Variables|Default|GetDataAsset))) (ScalableFloat|GetValueatLevel (Variables|/ItemizedPropSpawnerParent/Blueprints/DataClass_ItemizedPropSpawnerData.DataClass_ItemizedPropSpawnerData_C|GetDownwardTraceDistanceNear (Variables|Default|GetDataAsset))) (ScalableFloat|GetValueatLevel (Variables|/ItemizedPropSpawnerParent/Blueprints/DataClass_ItemizedPropSpawnerData.DataClass_ItemizedPropSpawnerData_C|GetDownwardTraceDistanceNearUp (Variables|Default|GetDataAsset))) (ScalableFloat|GetValueatLevel (Variables|/ItemizedPropSpawnerParent/Blueprints/DataClass_ItemizedPropSpawnerData.DataClass_ItemizedPropSpawnerData_C|GetDownwardTraceDistanceFar (Variables|Default|GetDataAsset))) (Variables|Default|GetClassesToIgnoreWhenTracing) (Variables|Default|GetClassesToIgnoreWhenTracingIfSmall) _proprelativetransforms _chunklocationtopropindexmap (ScalableFloat|GetValueatLevel (Variables|/ItemizedPropSpawnerParent/Blueprints/DataClass_ItemizedPropSpawnerData.DataClass_ItemizedPropSpawnerData_C|GetChunkRelativeRightVectorScale (Variables|Default|GetDataAsset))) (ScalableFloat|GetValueatLevel (Variables|/ItemizedPropSpawnerParent/Blueprints/DataClass_ItemizedPropSpawnerData.DataClass_ItemizedPropSpawnerData_C|GetGroundFindStartOffset (Variables|Default|GetDataAsset))) (ScalableFloat|GetValueatLevel (Variables|/ItemizedPropSpawnerParent/Blueprints/DataClass_ItemizedPropSpawnerData.DataClass_ItemizedPropSpawnerData_C|GetGroundFindEndOffset (Variables|Default|GetDataAsset))) (ScalableFloat|GetValueatLevel (Variables|/ItemizedPropSpawnerParent/Blueprints/DataClass_ItemizedPropSpawnerData.DataClass_ItemizedPropSpawnerData_C|GetGroundFindDistanceOffset (Variables|Default|GetDataAsset))) (ScalableFloat|GetValueatLevel (Variables|/ItemizedPropSpawnerParent/Blueprints/DataClass_ItemizedPropSpawnerData.DataClass_ItemizedPropSpawnerData_C|GetMaxSpawnDistBelowPlayer (Variables|Default|GetDataAsset))) (Variables|Default|GetPreviewMeshes) (ScalableFloat|GetValueatLevel (Variables|/ItemizedPropSpawnerParent/Blueprints/DataClass_ItemizedPropSpawnerData.DataClass_ItemizedPropSpawnerData_C|GetIceBlockPitch (Variables|Default|GetDataAsset))) (ScalableFloat|GetValueatLevel (Variables|/ItemizedPropSpawnerParent/Blueprints/DataClass_ItemizedPropSpawnerData.DataClass_ItemizedPropSpawnerData_C|GetBlockingIndestructibleObjectDist (Variables|Default|GetDataAsset))) (ScalableFloat|GetValueatLevel (Variables|/ItemizedPropSpawnerParent/Blueprints/DataClass_ItemizedPropSpawnerData.DataClass_ItemizedPropSpawnerData_C|GetMaxVerticalChunkSpacing (Variables|Default|GetDataAsset))) (ScalableFloat|GetValueatLevel (Variables|/ItemizedPropSpawnerParent/Blueprints/DataClass_ItemizedPropSpawnerData.DataClass_ItemizedPropSpawnerData_C|GetMinSupportingPropSize (Variables|Default|GetDataAsset))) (ScalableFloat|GetValueatLevel (Variables|/ItemizedPropSpawnerParent/Blueprints/DataClass_ItemizedPropSpawnerData.DataClass_ItemizedPropSpawnerData_C|GetTooCloseForHeightDistStep (Variables|Default|GetDataAsset))) (ScalableFloat|GetValueasInteger (Variables|/ItemizedPropSpawnerParent/Blueprints/DataClass_ItemizedPropSpawnerData.DataClass_ItemizedPropSpawnerData_C|GetTooCloseForHeightMaxIterations (Variables|Default|GetDataAsset))) (ScalableFloat|GetValueatLevel (Variables|/ItemizedPropSpawnerParent/Blueprints/DataClass_ItemizedPropSpawnerData.DataClass_ItemizedPropSpawnerData_C|GetTooCloseForHeightMinHeightAbovePlayer (Variables|Default|GetDataAsset))) (ScalableFloat|GetValueatLevel (Variables|/ItemizedPropSpawnerParent/Blueprints/DataClass_ItemizedPropSpawnerData.DataClass_ItemizedPropSpawnerData_C|GetTooCloseForHeightMaxHeightAbovePlayer (Variables|Default|GetDataAsset))) (ScalableFloat|GetValueatLevel (Variables|/ItemizedPropSpawnerParent/Blueprints/DataClass_ItemizedPropSpawnerData.DataClass_ItemizedPropSpawnerData_C|GetTooCloseForHeightMinDistFromPlayer (Variables|Default|GetDataAsset))) (ScalableFloat|GetValueatLevel (Variables|/ItemizedPropSpawnerParent/Blueprints/DataClass_ItemizedPropSpawnerData.DataClass_ItemizedPropSpawnerData_C|GetSphereTraceRadius (Variables|Default|GetDataAsset))) (ScalableFloat|GetValueasBool (Variables|/ItemizedPropSpawnerParent/Blueprints/DataClass_ItemizedPropSpawnerData.DataClass_ItemizedPropSpawnerData_C|GetAllowUnwalkableTargets (Variables|Default|GetDataAsset))) (ScalableFloat|GetValueasBool (Variables|/ItemizedPropSpawnerParent/Blueprints/DataClass_ItemizedPropSpawnerData.DataClass_ItemizedPropSpawnerData_C|GetAllowWaterTargets (Variables|Default|GetDataAsset))) (ScalableFloat|GetValueatLevel (Variables|/ItemizedPropSpawnerParent/Blueprints/DataClass_ItemizedPropSpawnerData.DataClass_ItemizedPropSpawnerData_C|GetTryCloserOnInvalidTargetDist (Variables|Default|GetDataAsset))) (ScalableFloat|GetValueatLevel (Variables|/ItemizedPropSpawnerParent/Blueprints/DataClass_ItemizedPropSpawnerData.DataClass_ItemizedPropSpawnerData_C|GetVerticalDistanceLimit (Variables|Default|GetDataAsset)))))

(fn ShouldSpawnIceRacingFX
  (return (CallFunction|ShouldSpawnWallAfterFallingActivation)))

(fn ShouldSpawnWallAfterFallingActivation
  (return (and (Variables|Default|GetbHasValidTarget) (not (GameplayTags|HasAnyMatchingGameplayTags (Ability|GetFortAbilitySystemComponent (Variables|Default|GetPlayer)) (Variables|Default|GetTagsThatCancelFallingSpawn))))))

(fn ShowReticle
  (bind _asfort_player_controller_zone (Utilities|Casting|CastToFortPlayerControllerZone (Game|GetInstigatorController)))
  (HUD|ResetHUDElementVisibilitytoSettings _asfort_player_controller_zone (Variables|Scope|GetReticleHUDElementTags)))

(fn ToggleMobilePreview
  (params (ShouldDisplayPreview :type Boolean))
  (CallFunction|OnPreviewingChanged (Math|Boolean|ANDBoolean ShouldDisplayPreview (TouchPlatform|IsUsingTouch (Fortnite|GetFortPlayerControllerfromActor (Variables|Default|GetPlayer))) (not (GameplayTags|HasMatchingGameplayTag (Variables|Default|GetPlayer) (Variables|Default|GetTag_Cooldown))))))

(fn TraceForSpawning
  (params (Start :type Vector) (End :type Vector))
  ;[1 0 0] This function has been nativized as NativeTraceForSpawning, please use that instead. If you need to make changes contact Brian Zaugg.
  (bind _outhits (Collision|MultiLineTraceByProfile :Start Start :End End :ProfileName "OverlapAllDynamic" :bTraceComplex false :DrawDebugType "ForOneFrame" :bIgnoreSelf true :TraceColor "(R=1.000000,G=0.000000,B=0.000000,A=1.000000)" :TraceHitColor "(R=0.000000,G=1.000000,B=0.000000,A=1.000000)" :DrawTime 5.0))
  (bind _hit (CallFunction|FindFirstValidHit _outhits))
  (return _hit (Utilities|IsValid (.HitActor (Collision|BreakHitResult _hit)) :!class /Script/Engine.KismetSystemLibrary) (.Location (Collision|BreakHitResult _hit))))

(fn ConstructionScript)
