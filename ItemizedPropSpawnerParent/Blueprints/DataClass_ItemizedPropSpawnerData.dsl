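; Data class for an itemized prop spawner that computes a launch velocity from an instigator actor.
; GetLaunchVelocity flattens and normalizes the instigator's forward vector, then adds a vertical component from the launch Z multiplier.
; It scales that unnormalized sum by either the vehicle or pawn launch amount, chosen by the IsVehicle flag, so the multiplier sets lift relative to horizontal speed.
; All three tuning values are read as scalable floats.
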(fn GetLaunchVelocity
  (params (Instigator :type Actor%20Object%20Reference) (IsVehicle :type Boolean))
  (return (* (Utilities|Operators|Add (Math|Vector|Normalize2D(Vector) (Transformation|GetActorForwardVector Instigator)) (Math|Vector|MakeVector 0.0 0.0 (ScalableFloat|GetValueatLevel (Variables|Launch|GetLaunchZMultiplier)))) (ScalableFloat|GetValueatLevel (select IsVehicle (Variables|Launch|GetVehicleLaunchAmount) (Variables|Launch|GetPawnLaunchAmount))))))
