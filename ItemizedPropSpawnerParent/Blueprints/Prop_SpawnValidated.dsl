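; Base spawnable prop for the itemized prop spawner that resolves overlaps with players, vehicles and other actors.
; BeginPlay runs a launch check, then a delayed non-launch check that restores ignored collision, plus optional periodic checks.
; Each check gathers overlaps from GetComponentsForOverlapCheck (empty here), then NativeCheckOverlappedActor evaluates actors sequentially, spreading work across ticks when needed.
; Results destroy overlapped actors or the prop, or trigger an impulse cue; older Blueprint check functions are marked nativized.
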(event EventBeginPlay
  (bind _returnvalue (Utilities|Time|GetGameTimeinSeconds))
  (bind _staticmeshcomponent (Variables|Debug|GetStaticMeshComponent))
  (bind _configdata (Variables|Default|GetConfigData))
  ;[0.06721 0.4801 0.5] WIP  material
  (Rendering|Material|SetCustomPrimitiveDataFloat _staticmeshcomponent 2 _returnvalue)
  (Rendering|Material|SetCustomPrimitiveDataFloat _staticmeshcomponent 3 _returnvalue)
  (CallFunction|SetupOverlappedActorCheckData)
  (if (ScalableFloat|GetValueasBool (Variables|/ItemizedPropSpawnerParent/Blueprints/DataClass_ItemizedPropSpawnerData.DataClass_ItemizedPropSpawnerData_C|GetOverlappingPlayersCheck_Enabled _configdata))
    (Utilities|Time|SetTimerbyEvent (delegate Custom|UpdatePlayerCheck) (ScalableFloat|GetValueatLevel (Variables|/ItemizedPropSpawnerParent/Blueprints/DataClass_ItemizedPropSpawnerData.DataClass_ItemizedPropSpawnerData_C|GetOverlappingPlayersCheck_DelayBetweenUpdates _configdata)) true false 0.0 (ScalableFloat|GetValueatLevel (Variables|/ItemizedPropSpawnerParent/Blueprints/DataClass_ItemizedPropSpawnerData.DataClass_ItemizedPropSpawnerData_C|GetOverlappingPlayersCheck_DelayBetweenUpdates _configdata))))
  (CallFunction|CheckforStuckPlayers true)
  (Utilities|Time|SetTimerbyEvent (delegate Custom|CheckForPlayersAndReEnableCollision) (Math|Random|RandomFloatinRange (ScalableFloat|GetValueatLevel (Variables|/ItemizedPropSpawnerParent/Blueprints/DataClass_ItemizedPropSpawnerData.DataClass_ItemizedPropSpawnerData_C|GetOverlappingPlayersCheck_DelayBeforeInitialCheck (Variables|Default|GetConfigData))) (+ (ScalableFloat|GetValueatLevel (Variables|/ItemizedPropSpawnerParent/Blueprints/DataClass_ItemizedPropSpawnerData.DataClass_ItemizedPropSpawnerData_C|GetOverlappingPlayersCheck_DelayBeforeInitialCheck (Variables|Default|GetConfigData))) 0.3))))

(event Custom|CheckForPlayersAndReEnableCollision
  (CallFunction|CheckforStuckPlayers false)
  (for _array_element (Variables|Default|GetComponentsSetToIgnoreCollision)
    (Collision|IgnoreActorwhenMoving _array_element self false)))

(event Custom|CheckForStuckPlayers
  (params (Launch :type Boolean))
  (if (not (Variables|Default|GetCheckInProgress))
    (Variables|Default|SetShouldLaunch Launch)
    (Variables|Default|SetCheckInProgress true)
    (Utilities|Array|Clear (Variables|Default|GetLastOverlappedActors))
    (Utilities|Array|Clear (Variables|Default|GetOverlapCheckComponents))
    (bind _output_get (Variables|Default|SetOverlapCheckComponents (CallFunction|GetComponentsforOverlapCheck)))
    (for _array_element _output_get
      (bind (_returnvalue_location _returnvalue_rotation _returnvalue_scale) (Math|Transform|BreakTransform (Transformation|GetWorldTransform _array_element)))
      (bind _outcomponents (Collision|ComponentOverlapComponents :Component _array_element :ComponentTransform (Math|Transform|MakeTransform (+ _returnvalue_location "0.000000,25.000000,0.000000") _returnvalue_rotation (* _returnvalue_scale "0.940000,0.900000,0.900000")) :ObjectTypes (Utilities|Array|MakeArray "ObjectTypeQuery3" "ObjectTypeQuery1" "ObjectTypeQuery2" "ObjectTypeQuery5")))
      (bind _output_get_2 (Variables|Default|SetOut%20Components _outcomponents))
      (for _array_element_2 _output_get_2
        (switch Utilities|FlowControl|Switch|SwitchonECollisionResponse (Collision|GetCollisionResponsetoChannel _array_element_2 "ECC_GameTraceChannel16")
          (:ECR_Ignore
            (switch Utilities|FlowControl|Switch|SwitchonECollisionEnabled (Collision|GetCollisionEnabled _array_element_2)
              (:PhysicsOnly
                _)
              (:QueryAndPhysics
                _)))
          (:ECR_Overlap
            _)
          (:ECR_Block
            _))
        (Utilities|Array|AddUnique (Variables|Default|GetLastOverlappedActors) (Components|GetOwner _array_element_2))))
    (Variables|Default|SetCurrentPlayerCheckIndex 0)
    (if (Utilities|Array|IsNotEmpty (Variables|Default|GetLastOverlappedActors))
      (Utilities|FlowControl|DelayUntilNextTick)
      (CallFunction|CheckNextOverlappedActor))))

(event Custom|UpdatePlayerCheck
  (CallFunction|CheckforStuckPlayers false))

(event EventOnDamagePlayEffects
  (params (Damage :type Float%20%28single-precision%29) (DamageTags :type Gameplay%20Tag%20Container%20Structure%20%28by%20ref%29) (Momentum :type Vector :default "0, 0, 0") (HitInfo :type Hit%20Result%20Structure%20%28by%20ref%29) (InstigatedBy :type Fort%20Pawn%20Object%20Reference) (DamageCauser :type Actor%20Object%20Reference) (EffectContext :type Gameplay%20Effect%20Context%20Handle%20Structure))
  (bind (_bblockinghit _binitialoverlap _time _distance _location _impactpoint _normal _impactnormal) (Collision|BreakHitResult HitInfo))
  ;[1 1 1] Impact Effects
  (Niagara|SpawnSystematLocation (Variables|Default|GetOnDamagedFX) _location (Math|Conversions|RotationFromXVector _impactnormal)))

(event Custom|CheckNextOverlappedActor
  (bind _currentplayercheckindex (Variables|Default|GetCurrentPlayerCheckIndex))
  (bind _lastoverlappedactors (Variables|Default|GetLastOverlappedActors))
  (bind _output (Utilities|Array|Get(acopy) _lastoverlappedactors _currentplayercheckindex))
  (bind _structout (Utilities|Struct|SetmembersinItemizedPropSpawnerOverlappedActorCheckData (Variables|Default|GetOverlappedActorCheckData) (Variables|Default|GetShouldLaunch) (Team|OnSameTeam (Variables|Actor|GetInstigator) _output)))
  (bind (_boutdestroyoverlappedactor _boutshouldwaitatick _boutdestroyprop) (ItemizedPropSpawner|NativeCheckOverlappedActor _structout _output (Variables|Default|GetComponentsSetToIgnoreCollision) (Variables|Default|GetShouldPlayImpulseCue)))
  (if _boutdestroyoverlappedactor
    (Actor|DestroyActor _output))
  (bind _output_get (Variables|Default|SetShouldDestroy _boutdestroyprop))
  (if (and _output_get (not (Variables|Default|GetShouldLaunch)))
    (CallFunction|FinishedChecking)
    (elif _boutshouldwaitatick
      (Utilities|FlowControl|DelayUntilNextTick)))
  (bind _result (Math|Integer|IncrementInt (Variables|Default|GetCurrentPlayerCheckIndex))
    (:%20%20
      (if (<= _result (Utilities|Array|LastIndex (Variables|Default|GetLastOverlappedActors)))
        (CallFunction|CheckNextOverlappedActor)
        (else
          (CallFunction|FinishedChecking))))))

(event Custom|Finished%20Checking
  (Utilities|FlowControl|Sequence
    (:then_0
      (if (Variables|Default|GetShouldDestroy)
        (Actor|DestroyActor)))
    (:then_1
      (if (Variables|Default|GetShouldPlayImpulseCue)
        (GameplayCue|ExecuteGameplayCue 0 (Variables|Default|GetGC_ImpulsedSomeone))))
    (:then_2
      (Variables|Default|SetCheckInProgress false))))

(fn CheckOverlappedActor
  (params (Overlapped%20Actor :type Object%20Reference))
  ;[1 0 0.01836 1 36] This function has been nativized as NativeCheckOverlappedActor, please use that instead. If you need to make changes contact Jacob Stone.
  (Utilities|FlowControl|Sequence
    (:then_0
      (if (CallFunction|ShouldActorBlockSpawn (Variables|GetOverlapped%20Actor))
        (LocalVariables|SetShould%20Destroy%20%28local%29 true)
        (elif (CallFunction|ShouldDestroyOverlappedActor (Variables|GetOverlapped%20Actor))
          (bind _asactor (Utilities|Casting|CastToActor (Variables|GetOverlapped%20Actor)))
          (Actor|DestroyActor _asactor)
          (else
            (bind _asfort_athena_vehicle (Utilities|Casting|CastToFortAthenaVehicle (Variables|GetOverlapped%20Actor))
              (:then
                (if (Variables|Default|GetShouldLaunch)
                  (Audio|BroadcastAbilityImpactatLocation 0 _asfort_athena_vehicle (Class|DataClassItemizedPropSpawnerData|GetLaunchVelocity (Variables|Default|GetConfigData) (Variables|Actor|GetInstigator) true) 777.0) ; Launch Vehicle\n(Node is disguised for anti-cheat reasons)
                  (FortPhysicsSimulationLibrary|SetImpactInstigatorforPhysicsObject _asfort_athena_vehicle (Variables|Actor|GetInstigator))
                  (if (Variables|Default|GetShouldLaunch)
                    (Utilities|Array|Add (Variables|Default|GetComponentsSetToIgnoreCollision) (Variables|/Script/Engine.Character|GetCapsuleComponent (LocalVariables|GetOverlappedActorAsPawn)))
                    (Variables|Default|SetShouldPlayImpulseCue true))))
              (:CastFailed
                (bind _asfort_pawn (Utilities|Casting|CastToFortPawn (Variables|GetOverlapped%20Actor)))
                (LocalVariables|SetOverlappedActorAsPawn _asfort_pawn)
                (if (Variables|Default|GetShouldLaunch)
                  (if (not (== (LocalVariables|GetOverlappedActorAsPawn) (Variables|Actor|GetInstigator)))
                    (Collision|IgnoreActorwhenMoving (Variables|/Script/Engine.Character|GetCapsuleComponent (LocalVariables|GetOverlappedActorAsPawn)) self true)
                    (Pawn|Character|LaunchCharacterJump (LocalVariables|GetOverlappedActorAsPawn) (Class|DataClassItemizedPropSpawnerData|GetLaunchVelocity (Variables|Default|GetConfigData) (Variables|Actor|GetInstigator) false) true true (or (Team|OnSameTeam (LocalVariables|GetOverlappedActorAsPawn) (Variables|Actor|GetInstigator)) (Class|AthenaFunctionLibrary|RowtoBool (Variables|Default|GetSF_ApplyFallImmunityToKnockbackVictims))) true)
                    (Utilities|Array|Add (Variables|Default|GetComponentsSetToIgnoreCollision) (Variables|/Script/Engine.Character|GetCapsuleComponent (LocalVariables|GetOverlappedActorAsPawn)))
                    (Variables|Default|SetShouldPlayImpulseCue true))
                  (else
                    (LocalVariables|SetShould%20Wait%20a%20Tick_ true)
                    (bind _outactors (Collision|CapsuleOverlapActors (Transformation|GetActorLocation _asfort_pawn) (* (Variables|/Script/Engine.CapsuleComponent|GetCapsuleRadius (Variables|/Script/Engine.Character|GetCapsuleComponent _asfort_pawn)) (ScalableFloat|GetValueatLevel (Variables|/ItemizedPropSpawnerParent/Blueprints/DataClass_ItemizedPropSpawnerData.DataClass_ItemizedPropSpawnerData_C|GetOverlappingPlayersCheck_CapsuleRadiusMultiplier (Variables|Default|GetConfigData)))) (* (Variables|/Script/Engine.CapsuleComponent|GetCapsuleHalfHeight (Variables|/Script/Engine.Character|GetCapsuleComponent _asfort_pawn)) (ScalableFloat|GetValueatLevel (Variables|/ItemizedPropSpawnerParent/Blueprints/DataClass_ItemizedPropSpawnerData.DataClass_ItemizedPropSpawnerData_C|GetOverlappingPlayersCheck_CapsuleHalfHeight (Variables|Default|GetConfigData)))) (Utilities|Array|MakeArray "ObjectTypeQuery1") "/ItemizedPropSpawnerParent/Blueprints/Prop_SpawnValidated.Prop_SpawnValidated_C"))
                    (for _array_element _outactors
                      (if (== _array_element self)
                        (LocalVariables|SetShould%20Destroy%20%28local%29 true)))))))))))
    (:then_1
      (return (LocalVariables|GetShould%20Wait%20a%20Tick_) (LocalVariables|GetShould%20Destroy%20%28local%29)))))

(fn GetComponentsForOverlapCheck
  (return 0))

(fn IsVehicleActive?
  (params (Vehicle :type Fort%20Vehicle%20Interface%20Interface))
  (return (or (Utilities|Array|IsNotEmpty (Vehicle|GetAllPassengers Vehicle)) (> (Math|Vector|VectorLength (Vehicle|GetVehicleVelocity Vehicle)) 0.1))))

(fn SetupOverlappedActorCheckData
  (Utilities|Struct|SetmembersinItemizedPropSpawnerOverlappedActorCheckData (Variables|Default|GetOverlappedActorCheckData) (Utilities|Array|MakeArray "ObjectTypeQuery1") "/ItemizedPropSpawnerParent/Blueprints/Prop_SpawnValidated.Prop_SpawnValidated_C" (Variables|Default|GetTagsToDestroyOverlappedActor) "(TagName=\"Item.SereneYard.BlockSpawn\")" (Class|DataClassItemizedPropSpawnerData|GetLaunchVelocity (Variables|Default|GetConfigData) (Variables|Actor|GetInstigator) true) (Class|DataClassItemizedPropSpawnerData|GetLaunchVelocity (Variables|Default|GetConfigData) (Variables|Actor|GetInstigator) false) (ScalableFloat|GetValueatLevel (Variables|/ItemizedPropSpawnerParent/Blueprints/DataClass_ItemizedPropSpawnerData.DataClass_ItemizedPropSpawnerData_C|GetOverlappingPlayersCheck_CapsuleRadiusMultiplier (Variables|Default|GetConfigData))) (ScalableFloat|GetValueatLevel (Variables|/ItemizedPropSpawnerParent/Blueprints/DataClass_ItemizedPropSpawnerData.DataClass_ItemizedPropSpawnerData_C|GetOverlappingPlayersCheck_CapsuleHalfHeight (Variables|Default|GetConfigData))) (Class|AthenaFunctionLibrary|RowtoBool (Variables|Default|GetSF_ApplyFallImmunityToKnockbackVictims))))

(fn ShouldActorBlockSpawn
  (params (Actor :type Object%20Reference))
  (bind (_asactor _bsuccess) (Utilities|Casting|CastToActor Actor))
  ;[1 0.03073 0.02078] This function has been nativized as NativeShouldActorBlockSpawn, please use that instead. If you need to make changes contact Jacob Stone.
  (if (Variables|Default|GetShouldLaunch)
    (return false)
    (else
      (bind _returnvalue (GameplayTags|HasMatchingGameplayTag(Message) Actor "(TagName=\"Item.SereneYard.BlockSpawn\")"))
      (return (or _returnvalue (select _bsuccess (not (ItemizedPropSpawner|IsWallAllowedtoOverlapActor _asactor)) false))))))

(fn ShouldDestroyOverlappedActor
  (params (Actor :type Object%20Reference))
  ;[1 0.02953 0] This function has been nativized as NativeShouldActorBlockSpawn, please use that instead. If you need to make changes contact Jacob Stone.
  (bind _returnvalue (GameplayTags|HasAnyMatchingGameplayTags(Message) Actor (Variables|Default|GetTagsToDestroyOverlappedActor)))
  (return _returnvalue))

(fn ConstructionScript
  (Utilities|IsValid (Variables|Default|GetCosmeticMesh)
    (:Is%20Valid
      (bind _returnvalue (AddComponent|NODE_AddStaticMeshComponent-0 :RelativeTransform (Variables|Default|GetCosmeticMeshRelativeTransform)))
      (Components|StaticMesh|SetStaticMesh _returnvalue (Variables|Default|GetCosmeticMesh))
      (Collision|SetCollisionEnabled _returnvalue "NoCollision"))))
